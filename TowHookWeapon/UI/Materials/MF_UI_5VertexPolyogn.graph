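# UI material function that draws a five-sided polygon from five corner inputs as a signed distance field, outputting Fill, Outline, SDF and SDF Outline.
# UVs are remapped to -1..1, and distance comes from the minimum squared clamped projection onto each edge, square-rooted.
# Per-edge crossing tests using FortUILibrary boolean operations produce ±1 factors whose product signs inside versus outside.
# Outline Thickness offsets the absolute distance, and Glow Min/Max set the inverted SmoothStep falloff.
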MaterialFunction /TowHookWeapon/UI/Materials/MF_UI_5VertexPolyogn
graph {nodes [
  n1 = NamedRerouteUsage()
  FunctionOutput(None=@n1) {outputName=SDF, sortPriority=2}
  n2 = FunctionInput(Preview=Constant() {r=0.05}) {
   bUsePreviewValueAsDefault=true
   description=Determines the thickness of the stroke when using one of the outline outputs.
   inputName=Outline Thickness
   inputType=FunctionInput_Vector2
   sortPriority=12
  }
  n3 = Subtract(A=Abs(None=@n1), B=@n2) {constB=0.1}
  FunctionOutput(None=@n3) {outputName=SDF Outline, sortPriority=3}
  n4 = FunctionInput(Preview=Constant()) {
   bUsePreviewValueAsDefault=true
   description=Determines glow falloff strength. Can not go below GlowMin value or results will invert
   inputName=Glow Min
   inputType=FunctionInput_Scalar
   sortPriority=10
  }
  n5 = FunctionInput(Preview=Constant()) {
   bUsePreviewValueAsDefault=true
   description=Determines glow falloff strength. Can not go below GlowMin value or results will invert
   inputName=Glow Max
   inputType=FunctionInput_Scalar
   sortPriority=11
  }
  FunctionOutput(None=OneMinus(None=SmoothStep(Min=@n4, Max=@n5, Value=@n1))) {outputName=Fill}
  FunctionOutput(None=OneMinus(None=SmoothStep(Min=@n4, Max=@n5, Value=@n3))) {outputName=Outline, sortPriority=1}
  n6 = FunctionInput(Preview=Constant2Vector() {r=1}) {
   bUsePreviewValueAsDefault=true
   description=Controls the position of the third corner.
   inputName=Corner 04
   inputType=FunctionInput_Vector2
   sortPriority=4
  }
  NamedRerouteDeclaration(None=@n6) {name=Vertex3, nodeColor=<rgba 0 0 0 1>}
  n7 = FunctionInput(Preview=Constant2Vector() {g=1}) {
   bUsePreviewValueAsDefault=true
   description=Controls the position of the third corner.
   inputName=Corner 05
   inputType=FunctionInput_Vector2
   sortPriority=5
  }
  NamedRerouteDeclaration(None=@n7) {name=Vertex4, nodeColor=<rgba 0 0 0 1>}
  n8 = SmoothStep(
   Min=MaterialFunctionCall(Float2=NamedRerouteUsage()) {materialFunction=&BreakOutFloat2Components}.G
   Value=ComponentMask(None=NamedRerouteUsage()) {g=true}
  )
  n9 = NamedRerouteUsage()
  n10 = NamedRerouteUsage()
  n11 = Ceil(None=SmoothStep(
    Min=MaterialFunctionCall(Float2=@n9) {materialFunction=&BreakOutFloat2Components}.G
    Value=ComponentMask(None=@n10) {g=true}
   ))
  n12 = MaterialFunctionCall(Float2=NamedRerouteUsage()) {materialFunction=&BreakOutFloat2Components}
  n13 = MaterialFunctionCall(Float2=NamedRerouteUsage()) {materialFunction=&BreakOutFloat2Components}
  n14 = MaterialFunctionCall(Float2=NamedRerouteUsage()) {materialFunction=&BreakOutFloat2Components}
  n15 = MaterialFunctionCall(Float2=NamedRerouteUsage()) {materialFunction=&BreakOutFloat2Components}
  n16 = Add(
   A=Add(A=Ceil(None=@n8), B=OneMinus(None=@n11))
   B=MaterialFunctionCall!(A=Multiply(A=@n12.R, B=@n13.G), B=Multiply(A=@n14.G, B=@n15.R)) {materialFunction=/FortUILibrary/MaterialFunctions/Basics/MF_UI_BooleanOperations}.IsGreater
  )
  n17 = Add(
   A=MaterialFunctionCall(A=@n16, B=Constant() {r=3}) {materialFunction=&MF_UI_BooleanOperations}.IsGreaterOrEqual
   B=MaterialFunctionCall(A=@n16, B=Constant()) {materialFunction=&MF_UI_BooleanOperations}.IsLessOrEqual
  )
  n18 = Constant() {r=1}
  NamedRerouteDeclaration(None=LinearInterpolate(Alpha=MaterialFunctionCall(A=@n17, B=@n18) {materialFunction=&MF_UI_BooleanOperations}.IsGreaterOrEqual) {constA=1, constB=-1}) {name=s5, nodeColor=<rgba 0 0 0 1>}
  NamedRerouteDeclaration(None=Subtract(A=NamedRerouteUsage(), B=NamedRerouteUsage())) {name=w4, nodeColor=<rgba 0 0 0 1>}
  NamedRerouteDeclaration(None=Subtract(A=NamedRerouteUsage(), B=NamedRerouteUsage())) {name=w5, nodeColor=<rgba 0 0 0 1>}
  n19 = NamedRerouteUsage()
  n20 = NamedRerouteUsage()
  n21 = Min(A=Min(A=NamedRerouteUsage(), B=DotProduct(A=@n19, B=@n19)), B=DotProduct(A=@n20, B=@n20))
  n22 = NamedRerouteUsage()
  n23 = DotProduct(A=@n22, B=@n22)
  n24 = NamedRerouteUsage()
  n25 = NamedRerouteUsage()
  n26 = Min(A=Min(A=Min(A=@n21, B=@n23), B=DotProduct(A=@n24, B=@n24)), B=DotProduct(A=@n25, B=@n25))
  NamedRerouteDeclaration(None=Multiply(A=SquareRoot(None=@n26), B=NamedRerouteUsage())) {name=SDF Shape, nodeColor=<rgba 0 0 0 1>}
  n27 = Subtract(A=NamedRerouteUsage(), B=NamedRerouteUsage())
  NamedRerouteDeclaration(None=DotProduct(A=@n27, B=@n27)) {name=Distance, nodeColor=<rgba 0 0 0 1>}
  NamedRerouteDeclaration(None=Subtract(A=NamedRerouteUsage(), B=NamedRerouteUsage())) {name=w1, nodeColor=<rgba 0 0 0 1>}
  NamedRerouteDeclaration(None=Subtract(A=NamedRerouteUsage(), B=NamedRerouteUsage())) {name=w2, nodeColor=<rgba 0 0 0 1>}
  NamedRerouteDeclaration(None=Subtract(A=NamedRerouteUsage(), B=NamedRerouteUsage())) {name=w3, nodeColor=<rgba 0 0 0 1>}
  n28 = FunctionInput(Preview=TextureCoordinate()) {
   bUsePreviewValueAsDefault=true
   description=Controls the UVs of the circle with the TextureCoordinate node plugged in. Can be use to manipulate the X/Y coordinates as well.
   inputName=UVs
   inputType=FunctionInput_Vector2
  }
  NamedRerouteDeclaration(None=ConstantBiasScale(None=@n28) {bias=-0.5, scale=2}) {name="UV's(-1,1)", nodeColor=<rgba 0 0 0 1>}
  n29 = FunctionInput(Preview=Constant2Vector() {r=-1}) {
   bUsePreviewValueAsDefault=true
   description=Controls the position of the first corner.
   inputName=Corner 01
   inputType=FunctionInput_Vector2
   sortPriority=1
  }
  NamedRerouteDeclaration(None=@n29) {name=Vertex0, nodeColor=<rgba 0 0 0 1>}
  n30 = FunctionInput(Preview=Constant2Vector() {g=-1, r=-1}) {
   bUsePreviewValueAsDefault=true
   description=Controls the position of the second corner.
   inputName=Corner 02
   inputType=FunctionInput_Vector2
   sortPriority=2
  }
  NamedRerouteDeclaration(None=@n30) {name=Vertex1, nodeColor=<rgba 0 0 0 1>}
  n31 = FunctionInput(Preview=Constant2Vector() {g=-1, r=1}) {
   bUsePreviewValueAsDefault=true
   description=Controls the position of the third corner.
   inputName=Corner 03
   inputType=FunctionInput_Vector2
   sortPriority=3
  }
  NamedRerouteDeclaration(None=@n31) {name=Vertex2, nodeColor=<rgba 0 0 0 1>}
  NamedRerouteDeclaration(None=Subtract(A=NamedRerouteUsage(), B=NamedRerouteUsage())) {name=e1, nodeColor=<rgba 0 0 0 1>}
  n32 = Multiply(A=Multiply(A=NamedRerouteUsage(), B=NamedRerouteUsage()), B=NamedRerouteUsage())
  n33 = NamedRerouteUsage()
  NamedRerouteDeclaration(None=Multiply(A=Multiply(A=@n32, B=@n33), B=NamedRerouteUsage())) {name=Inner Shape, nodeColor=<rgba 0 0 0 1>}
  NamedRerouteDeclaration(None=Subtract(A=NamedRerouteUsage(), B=NamedRerouteUsage())) {name=e2, nodeColor=<rgba 0 0 0 1>}
  NamedRerouteDeclaration(None=Subtract(A=NamedRerouteUsage(), B=NamedRerouteUsage())) {name=e3, nodeColor=<rgba 0 0 0 1>}
  n34 = NamedRerouteUsage()
  n35 = NamedRerouteUsage()
  n36 = NamedRerouteUsage()
  n37 = NamedRerouteUsage()
  n38 = Saturate(None=Divide(A=DotProduct(A=@n35, B=@n36), B=DotProduct(A=@n37, B=@n37)))
  NamedRerouteDeclaration(None=Subtract(A=NamedRerouteUsage(), B=Multiply(A=@n34, B=@n38))) {name=b1, nodeColor=<rgba 0 0 0 1>}
  n39 = NamedRerouteUsage()
  n40 = NamedRerouteUsage()
  n41 = NamedRerouteUsage()
  n42 = NamedRerouteUsage()
  n43 = Saturate(None=Divide(A=DotProduct(A=@n40, B=@n41), B=DotProduct(A=@n42, B=@n42)))
  NamedRerouteDeclaration(None=Subtract(A=NamedRerouteUsage(), B=Multiply(A=@n39, B=@n43))) {name=b2, nodeColor=<rgba 0 0 0 1>}
  n44 = NamedRerouteUsage()
  n45 = NamedRerouteUsage()
  n46 = NamedRerouteUsage()
  n47 = NamedRerouteUsage()
  n48 = Saturate(None=Divide(A=DotProduct(A=@n45, B=@n46), B=DotProduct(A=@n47, B=@n47)))
  NamedRerouteDeclaration(None=Subtract(A=NamedRerouteUsage(), B=Multiply(A=@n44, B=@n48))) {name=b3, nodeColor=<rgba 0 0 0 1>}
  n49 = SmoothStep(
   Min=MaterialFunctionCall(Float2=NamedRerouteUsage()) {materialFunction=&BreakOutFloat2Components}.G
   Value=ComponentMask(None=NamedRerouteUsage()) {g=true}
  )
  n50 = NamedRerouteUsage()
  n51 = NamedRerouteUsage()
  n52 = Ceil(None=SmoothStep(
    Min=MaterialFunctionCall(Float2=@n50) {materialFunction=&BreakOutFloat2Components}.G
    Value=ComponentMask(None=@n51) {g=true}
   ))
  n53 = MaterialFunctionCall(Float2=NamedRerouteUsage()) {materialFunction=&BreakOutFloat2Components}
  n54 = MaterialFunctionCall(Float2=NamedRerouteUsage()) {materialFunction=&BreakOutFloat2Components}
  n55 = MaterialFunctionCall(Float2=NamedRerouteUsage()) {materialFunction=&BreakOutFloat2Components}
  n56 = MaterialFunctionCall(Float2=NamedRerouteUsage()) {materialFunction=&BreakOutFloat2Components}
  n57 = Add(
   A=Add(A=Ceil(None=@n49), B=OneMinus(None=@n52))
   B=MaterialFunctionCall(A=Multiply(A=@n53.R, B=@n54.G), B=Multiply(A=@n55.G, B=@n56.R)) {materialFunction=&MF_UI_BooleanOperations}.IsGreater
  )
  n58 = Add(
   A=MaterialFunctionCall(A=@n57, B=Constant() {r=3}) {materialFunction=&MF_UI_BooleanOperations}.IsGreaterOrEqual
   B=MaterialFunctionCall(A=@n57, B=Constant()) {materialFunction=&MF_UI_BooleanOperations}.IsLessOrEqual
  )
  n59 = Constant() {r=1}
  NamedRerouteDeclaration(None=LinearInterpolate(Alpha=MaterialFunctionCall(A=@n58, B=@n59) {materialFunction=&MF_UI_BooleanOperations}.IsGreaterOrEqual) {constA=1, constB=-1}) {name=s1, nodeColor=<rgba 0 0 0 1>}
  n60 = SmoothStep(
   Min=MaterialFunctionCall(Float2=NamedRerouteUsage()) {materialFunction=&BreakOutFloat2Components}.G
   Value=ComponentMask(None=NamedRerouteUsage()) {g=true}
  )
  n61 = NamedRerouteUsage()
  n62 = NamedRerouteUsage()
  n63 = Ceil(None=SmoothStep(
    Min=MaterialFunctionCall(Float2=@n61) {materialFunction=&BreakOutFloat2Components}.G
    Value=ComponentMask(None=@n62) {g=true}
   ))
  n64 = MaterialFunctionCall(Float2=NamedRerouteUsage()) {materialFunction=&BreakOutFloat2Components}
  n65 = MaterialFunctionCall(Float2=NamedRerouteUsage()) {materialFunction=&BreakOutFloat2Components}
  n66 = MaterialFunctionCall(Float2=NamedRerouteUsage()) {materialFunction=&BreakOutFloat2Components}
  n67 = MaterialFunctionCall(Float2=NamedRerouteUsage()) {materialFunction=&BreakOutFloat2Components}
  n68 = Add(
   A=Add(A=Ceil(None=@n60), B=OneMinus(None=@n63))
   B=MaterialFunctionCall(A=Multiply(A=@n64.R, B=@n65.G), B=Multiply(A=@n66.G, B=@n67.R)) {materialFunction=&MF_UI_BooleanOperations}.IsGreater
  )
  n69 = Add(
   A=MaterialFunctionCall(A=@n68, B=Constant() {r=3}) {materialFunction=&MF_UI_BooleanOperations}.IsGreaterOrEqual
   B=MaterialFunctionCall(A=@n68, B=Constant()) {materialFunction=&MF_UI_BooleanOperations}.IsLessOrEqual
  )
  n70 = Constant() {r=1}
  NamedRerouteDeclaration(None=LinearInterpolate(Alpha=MaterialFunctionCall(A=@n69, B=@n70) {materialFunction=&MF_UI_BooleanOperations}.IsGreaterOrEqual) {constA=1, constB=-1}) {name=s2, nodeColor=<rgba 0 0 0 1>}
  n71 = SmoothStep(
   Min=MaterialFunctionCall(Float2=NamedRerouteUsage()) {materialFunction=&BreakOutFloat2Components}.G
   Value=ComponentMask(None=NamedRerouteUsage()) {g=true}
  )
  n72 = NamedRerouteUsage()
  n73 = NamedRerouteUsage()
  n74 = Ceil(None=SmoothStep(
    Min=MaterialFunctionCall(Float2=@n72) {materialFunction=&BreakOutFloat2Components}.G
    Value=ComponentMask(None=@n73) {g=true}
   ))
  n75 = MaterialFunctionCall(Float2=NamedRerouteUsage()) {materialFunction=&BreakOutFloat2Components}
  n76 = MaterialFunctionCall(Float2=NamedRerouteUsage()) {materialFunction=&BreakOutFloat2Components}
  n77 = MaterialFunctionCall(Float2=NamedRerouteUsage()) {materialFunction=&BreakOutFloat2Components}
  n78 = MaterialFunctionCall(Float2=NamedRerouteUsage()) {materialFunction=&BreakOutFloat2Components}
  n79 = Add(
   A=Add(A=Ceil(None=@n71), B=OneMinus(None=@n74))
   B=MaterialFunctionCall(A=Multiply(A=@n75.R, B=@n76.G), B=Multiply(A=@n77.G, B=@n78.R)) {materialFunction=&MF_UI_BooleanOperations}.IsGreater
  )
  n80 = Add(
   A=MaterialFunctionCall(A=@n79, B=Constant() {r=3}) {materialFunction=&MF_UI_BooleanOperations}.IsGreaterOrEqual
   B=MaterialFunctionCall(A=@n79, B=Constant()) {materialFunction=&MF_UI_BooleanOperations}.IsLessOrEqual
  )
  n81 = Constant() {r=1}
  NamedRerouteDeclaration(None=LinearInterpolate(Alpha=MaterialFunctionCall(A=@n80, B=@n81) {materialFunction=&MF_UI_BooleanOperations}.IsGreaterOrEqual) {constA=1, constB=-1}) {name=s3, nodeColor=<rgba 0 0 0 1>}
  NamedRerouteDeclaration(None=Subtract(A=NamedRerouteUsage(), B=NamedRerouteUsage())) {name=e4, nodeColor=<rgba 0 0 0 1>}
  NamedRerouteDeclaration(None=Subtract(A=NamedRerouteUsage(), B=NamedRerouteUsage())) {name=e5, nodeColor=<rgba 0 0 0 1>}
  n82 = NamedRerouteUsage()
  n83 = NamedRerouteUsage()
  n84 = NamedRerouteUsage()
  n85 = NamedRerouteUsage()
  n86 = Saturate(None=Divide(A=DotProduct(A=@n83, B=@n84), B=DotProduct(A=@n85, B=@n85)))
  NamedRerouteDeclaration(None=Subtract(A=NamedRerouteUsage(), B=Multiply(A=@n82, B=@n86))) {name=b4, nodeColor=<rgba 0 0 0 1>}
  n87 = NamedRerouteUsage()
  n88 = NamedRerouteUsage()
  n89 = NamedRerouteUsage()
  n90 = NamedRerouteUsage()
  n91 = Saturate(None=Divide(A=DotProduct(A=@n88, B=@n89), B=DotProduct(A=@n90, B=@n90)))
  NamedRerouteDeclaration(None=Subtract(A=NamedRerouteUsage(), B=Multiply(A=@n87, B=@n91))) {name=b5, nodeColor=<rgba 0 0 0 1>}
  n92 = SmoothStep(
   Min=MaterialFunctionCall(Float2=NamedRerouteUsage()) {materialFunction=&BreakOutFloat2Components}.G
   Value=ComponentMask(None=NamedRerouteUsage()) {g=true}
  )
  n93 = NamedRerouteUsage()
  n94 = NamedRerouteUsage()
  n95 = Ceil(None=SmoothStep(
    Min=MaterialFunctionCall(Float2=@n93) {materialFunction=&BreakOutFloat2Components}.G
    Value=ComponentMask(None=@n94) {g=true}
   ))
  n96 = MaterialFunctionCall(Float2=NamedRerouteUsage()) {materialFunction=&BreakOutFloat2Components}
  n97 = MaterialFunctionCall(Float2=NamedRerouteUsage()) {materialFunction=&BreakOutFloat2Components}
  n98 = MaterialFunctionCall(Float2=NamedRerouteUsage()) {materialFunction=&BreakOutFloat2Components}
  n99 = MaterialFunctionCall(Float2=NamedRerouteUsage()) {materialFunction=&BreakOutFloat2Components}
  n100 = Add(
   A=Add(A=Ceil(None=@n92), B=OneMinus(None=@n95))
   B=MaterialFunctionCall(A=Multiply(A=@n96.R, B=@n97.G), B=Multiply(A=@n98.G, B=@n99.R)) {materialFunction=&MF_UI_BooleanOperations}.IsGreater
  )
  n101 = Add(
   A=MaterialFunctionCall(A=@n100, B=Constant() {r=3}) {materialFunction=&MF_UI_BooleanOperations}.IsGreaterOrEqual
   B=MaterialFunctionCall(A=@n100, B=Constant()) {materialFunction=&MF_UI_BooleanOperations}.IsLessOrEqual
  )
  n102 = Constant() {r=1}
  NamedRerouteDeclaration(None=LinearInterpolate(Alpha=MaterialFunctionCall(A=@n101, B=@n102) {materialFunction=&MF_UI_BooleanOperations}.IsGreaterOrEqual) {constA=1, constB=-1}) {name=s4, nodeColor=<rgba 0 0 0 1>}
 ]}
