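# Translucent UI material for the Tow Hook weapon's attachment reticle, drawing polygon markers, point-of-attachment (POA) texture shapes, a ring and a center dot.
# Five-vertex polygon SDFs, one mirrored and rotated plus one on top, merge via Min into fill and outer-glow masks.
# The rotatable reticle texture supplies POA fill and glow, with a three-segment radial reveal mask.
# Named reroutes carry the masks into opacity and color blends, with MorphValue driving a shape blend.
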Material /TowHookWeapon/UI/Materials/M_UI_TowHookAttachment
graph {
 properties {
  bAutomaticallySetUsageInEditor=false
  blendMode=BLEND_Translucent
  materialDomain=MD_UI
  pixelDepthOffsetMode=PDOM_Legacy
  usedShadingModels=Default Lit
 }
 nodes [
  Add(
   A=LinearInterpolate(A=NamedRerouteUsage(), B=NamedRerouteUsage(), Alpha=ScalarParameter() {parameterName=MorphValue}) {constAlpha=1}
   B=NamedRerouteUsage()
  )
  n1 = NamedRerouteUsage()
  n2 = Multiply(A=Add(A=NamedRerouteUsage(), B=NamedRerouteUsage()), B=@n1)
  n3 = ScalarParameter() {defaultValue=1, group=Polygon, parameterName=Polygon Opacity}
  n4 = NamedRerouteUsage()
  n5 = NamedRerouteUsage()
  n6 = Saturate(None=Multiply(A=Add(A=@n4, B=@n5), B=NamedRerouteUsage()))
  n7 = ScalarParameter() {defaultValue=1, group=POA, parameterName=POA Opacity}
  Add(A=Add(A=Multiply(A=@n2, B=@n3), B=Multiply(A=@n6, B=@n7)), B=NamedRerouteUsage())
  Constant2Vector() {g=-0.005, r=0.001}
  Constant() {r=0.38}
  Frac(None=Multiply(A=Time()) {constB=0.2})
  Frac(None=Multiply(A=Time()) {constB=0.5})
  LinearInterpolate(A=Constant3Vector() {constant=<rgba 1 1 1 0>}, B=Constant3Vector() {constant=<rgba 0.6016 0.6016 0.6016 0>}, Alpha=NamedRerouteUsage())
  n8 = Constant3Vector()
  n9 = Constant3Vector() {constant=<rgba 1 1 1 0>}
  n10 = Multiply(A=NamedRerouteUsage(), B=@n1)
  n11 = Constant3Vector() {constant=<rgba 1 1 1 0>}
  n12 = NamedRerouteUsage()
  LinearInterpolate(
   A=Add(A=LinearInterpolate(A=@n8, B=@n9, Alpha=@n10), B=LinearInterpolate(A=@n8, B=@n11, Alpha=@n12))
   B=Constant3Vector() {constant=<rgba 1 1 1 1>}
   Alpha=NamedRerouteUsage()
  )
  MaterialFunctionCall() {materialFunction=&DebugScalarValues}
  n13 = MaterialFunctionCall(Size=ScalarParameter() {defaultValue=0.06, parameterName=MiddleCircleSize}) {materialFunction=&MF_UI_Circle}
  NamedRerouteDeclaration(None=OneMinus(None=SmoothStep(Value=@n13.SDF) {constMax=0.01})) {name=MiddleCircle, nodeColor=<rgba 0 0 0 1>}
  n14 = TextureSample(UVs=MaterialFunctionCall(Texture Scale=Constant() {r=0.6}) {materialFunction=&ScaleUVsByCenter}.UVs) {samplerSource=SSM_Clamp_WorldGroupSettings, texture=/TowHookWeapon/UI/Textures/T_UI_Reticle_TowHook}
  n15 = ScalarParameter() {defaultValue=1, parameterName=POA_RevealSwitch}
  n16 = MaterialFunctionCall!(
   UV=MaterialFunctionCall(UVs=TextureCoordinate(), "Rotation Angle (0-1)"=Constant()) {materialFunction=&CustomRotator}.Rotated Values
   Segment Count=Constant() {r=3}
   Segment Spacing=SmoothStep(Value=OneMinus(None=@n15))
   Fixed Spacing
  ) {materialFunction=/FortUILibrary/MaterialFunctions/Basics/MF_UI_RadialSegments}
  NamedRerouteDeclaration(None=Multiply(A=SmoothStep(Value=@n14.G) {constMax=0.5, constMin=0.4}, B=Multiply(B=@n16.Segmented Triangles Fill) {constA=1})) {name=POA, nodeColor=<rgba 1 0 0.3984 1>}
  n17 = ScalarParameter() {defaultValue=0.156, parameterName=Point3_x}
  NamedRerouteDeclaration(None=AppendVector(A=Multiply(A=@n17) {constB=-1}, B=ScalarParameter() {defaultValue=-0.8, parameterName=Point3_y})) {name=Point4_1, nodeColor=<rgba 0 0 0 1>}
  n18 = MaterialFunctionCall!(
   UV=TextureCoordinate()
   Offset=AppendVector(A=ScalarParameter() {defaultValue=0.224, parameterName=OffsetX}, B=ScalarParameter() {defaultValue=-0.296, parameterName=OffsetY})
   Rotation
   Scale=Constant2Vector() {g=0.25, r=0.5}
   MirrorY
  ) {materialFunction=/FortUILibrary/MaterialFunctions/Basics/MF_UI_Symmetry_2x1_DEPRECATED}
  n19 = ScalarParameter() {defaultValue=0.375, parameterName=Angle}
  n20 = TextureCoordinate()
  n21 = AppendVector(
   A=ComponentMask(None=@n20) {r=true}
   B=Add(A=ComponentMask(None=@n20) {g=true}, B=ScalarParameter() {defaultValue=0.38, parameterName=OffsetY_Top})
  )
  n22 = Constant() {r=0.25}
  n23 = Min(
   A=MaterialFunctionCall!(
    UVs=MaterialFunctionCall(UVs=@n18.Result, "Rotation Angle (0-1)"=@n19) {materialFunction=&CustomRotator}.Rotated Values
    Corner 01=NamedRerouteUsage()
    Corner 02=NamedRerouteUsage()
    Corner 03=NamedRerouteUsage()
    Corner 04=NamedRerouteUsage()
    Corner 05=NamedRerouteUsage()
    Glow Min
    Glow Max
    Outline Thickness
   ) {materialFunction=&/MF_UI_5VertexPolyogn}.SDF
   B=MaterialFunctionCall(
    UVs=MaterialFunctionCall(UVs=@n21, Texture Scale=@n22) {materialFunction=&ScaleUVsByCenter}.UVs
    Corner 01=NamedRerouteUsage()
    Corner 02=NamedRerouteUsage()
    Corner 03=NamedRerouteUsage()
    Corner 04=NamedRerouteUsage()
    Corner 05=NamedRerouteUsage()
   ) {materialFunction=&MF_UI_5VertexPolyogn}.SDF
  )
  NamedRerouteDeclaration(None=OneMinus(None=SmoothStep(Value=@n23) {constMax=0.05})) {name=Polygon, nodeColor=<rgba 0 0.1094 1 1>}
  n24 = ScalarParameter() {defaultValue=0.244, parameterName=Point2_x}
  NamedRerouteDeclaration(None=AppendVector(A=Multiply(A=@n24) {constB=-1}, B=ScalarParameter() {defaultValue=0.564, parameterName=Point2_y})) {name=Point5_1, nodeColor=<rgba 0 0 0 1>}
  n25 = TextureCoordinate()
  n26 = AppendVector(A=ScalarParameter() {defaultValue=0.224, parameterName=OffsetX}, B=ScalarParameter() {defaultValue=-0.296, parameterName=OffsetY})
  n27 = Constant2Vector() {g=0.25, r=0.5}
  n28 = MaterialFunctionCall(
   UVs=MaterialFunctionCall(
    UVs=MaterialFunctionCall(UV=@n25, Offset=@n26, Scale=@n27) {materialFunction=&MF_UI_Symmetry_2x1_DEPRECATED}.Result
    "Rotation Angle (0-1)"=ScalarParameter() {defaultValue=0.375, parameterName=Angle}
   ) {materialFunction=&CustomRotator}.Rotated Values
   Corner 01=NamedRerouteUsage()
   Corner 02=NamedRerouteUsage()
   Corner 03=NamedRerouteUsage()
   Corner 04=NamedRerouteUsage()
   Corner 05=NamedRerouteUsage()
  ) {materialFunction=&MF_UI_5VertexPolyogn}
  n29 = TextureCoordinate()
  n30 = ComponentMask(None=@n29) {r=true}
  n31 = Add(A=ComponentMask(None=@n29) {g=true}, B=ScalarParameter() {defaultValue=0.38, parameterName=OffsetY_Top})
  n32 = MaterialFunctionCall(
   UVs=MaterialFunctionCall(UVs=AppendVector(A=@n30, B=@n31), Texture Scale=Constant() {r=0.25}) {materialFunction=&ScaleUVsByCenter}.UVs
   Corner 01=NamedRerouteUsage()
   Corner 02=NamedRerouteUsage()
   Corner 03=NamedRerouteUsage()
   Corner 04=NamedRerouteUsage()
   Corner 05=NamedRerouteUsage()
  ) {materialFunction=&MF_UI_5VertexPolyogn}
  n33 = Min(A=@n28.SDF, B=@n32.SDF)
  NamedRerouteDeclaration(None=OneMinus(None=SmoothStep(Value=@n33) {constMax=0.05})) {name=Polygon_1, nodeColor=<rgba 0 0.1094 1 1>}
  n34 = MaterialFunctionCall(Size=ScalarParameter() {defaultValue=0.06, parameterName=MiddleCircleSize}) {materialFunction=&MF_UI_Circle}
  NamedRerouteDeclaration(None=OneMinus(None=SmoothStep(Value=@n34.SDF) {constMax=0.01})) {name=MiddleCircle_1, nodeColor=<rgba 0 0 0 1>}
  n35 = SmoothStep(Value=MaterialFunctionCall(
    Size=ScalarParameter() {defaultValue=0.204, parameterName=RingSize}
    OutlineThickness=ScalarParameter() {defaultValue=0.016, parameterName=RingThickness}
   ) {materialFunction=&MF_UI_Circle}.SDFOutline) {constMax=0.01}
  NamedRerouteDeclaration(None=Multiply(A=OneMinus(None=@n35), B=ScalarParameter() {parameterName=RingOpacity})) {name=Ring, nodeColor=<rgba 0 1 0.2969 1>}
  n36 = TextureCoordinate()
  n37 = ScalarParameter() {parameterName=RotationAngle}
  n38 = TextureSample(UVs=MaterialFunctionCall(
    UVs=MaterialFunctionCall(UVs=@n36, "Rotation Angle (0-1)"=@n37) {materialFunction=&CustomRotator}.Rotated Values
    Texture Scale=ScalarParameter() {defaultValue=0.6, group=POA, parameterName=POA Scale}
   ) {materialFunction=&ScaleUVsByCenter}.UVs) {samplerSource=SSM_Clamp_WorldGroupSettings, texture=&T_UI_Reticle_TowHook}
  n39 = SmoothStep(Value=@n38.G) {constMax=0.4}
  NamedRerouteDeclaration(None=Multiply(A=@n39, B=SmoothStep(Value=@n38.G) {constMax=0.5, constMin=0.4})) {name=POA_1, nodeColor=<rgba 1 0 0.3984 1>}
  n40 = Min(A=@n28.SDF, B=@n32.SDF)
  NamedRerouteDeclaration(None=Multiply(A=SmoothStep(Value=@n40) {constMax=0, constMin=0.3}, B=ScalarParameter() {defaultValue=1, group=Shared, parameterName=OuterGlow_Opacity}) {constB=0.8}) {name=Polygon_OuterGlow, nodeColor=<rgba 0 0.3112 1 1>}
  NamedRerouteDeclaration(None=Multiply(A=@n39, B=ScalarParameter() {defaultValue=1, group=Shared, parameterName=OuterGlow_Opacity}) {constB=0.8}) {name=POA_OuterGlow, nodeColor=<rgba 1 0 0.7711 1>}
  n41 = MaterialFunctionCall(UVs=TextureCoordinate(), "Rotation Angle (0-1)"=Constant() {r=1}) {materialFunction=&CustomRotator}
  n42 = Constant() {r=3}
  n43 = SmoothStep(Value=OneMinus(None=ScalarParameter() {defaultValue=1, parameterName=POA_RevealSwitch}))
  NamedRerouteDeclaration(None=Multiply(B=MaterialFunctionCall(UV=@n41.Rotated Values, Segment Count=@n42, Segment Spacing=@n43) {materialFunction=&MF_UI_RadialSegments}.Segmented Triangles Fill) {constA=1}) {name=Segmented Mask, nodeColor=<rgba 1 0 0.7969 1>}
  NamedRerouteDeclaration(None=AppendVector(A=ScalarParameter() {parameterName=Point1_x}, B=ScalarParameter() {defaultValue=0.792, parameterName=Point1_y})) {name=Point1_1, nodeColor=<rgba 0 0 0 1>}
  n44 = ScalarParameter() {defaultValue=0.4487, parameterName=Polygone Mask Size}
  NamedRerouteDeclaration(None=SmoothStep(Value=MaterialFunctionCall(Size=@n44) {materialFunction=&MF_UI_Circle}.SDF) {constMax=0.01}) {name=Polygon_mask, nodeColor=<rgba 0 0 0 1>}
  NamedRerouteDeclaration(None=AppendVector(A=ScalarParameter() {defaultValue=0.244, parameterName=Point2_x}, B=ScalarParameter() {defaultValue=0.564, parameterName=Point2_y})) {name=Point2_1, nodeColor=<rgba 0 0 0 1>}
  NamedRerouteDeclaration(None=AppendVector(A=ScalarParameter() {parameterName=Point1_x}, B=ScalarParameter() {defaultValue=0.792, parameterName=Point1_y})) {name=Point1, nodeColor=<rgba 0 0 0 1>}
  NamedRerouteDeclaration(None=AppendVector(A=ScalarParameter() {defaultValue=0.244, parameterName=Point2_x}, B=ScalarParameter() {defaultValue=0.564, parameterName=Point2_y})) {name=Point2, nodeColor=<rgba 0 0 0 1>}
  NamedRerouteDeclaration(None=AppendVector(A=ScalarParameter() {defaultValue=0.156, parameterName=Point3_x}, B=ScalarParameter() {defaultValue=-0.8, parameterName=Point3_y})) {name=Point3, nodeColor=<rgba 0 0 0 1>}
  n45 = ScalarParameter() {defaultValue=0.156, parameterName=Point3_x}
  NamedRerouteDeclaration(None=AppendVector(A=Multiply(A=@n45) {constB=-1}, B=ScalarParameter() {defaultValue=-0.8, parameterName=Point3_y})) {name=Point4, nodeColor=<rgba 0 0 0 1>}
  n46 = ScalarParameter() {defaultValue=0.244, parameterName=Point2_x}
  NamedRerouteDeclaration(None=AppendVector(A=Multiply(A=@n46) {constB=-1}, B=ScalarParameter() {defaultValue=0.564, parameterName=Point2_y})) {name=Point5, nodeColor=<rgba 0 0 0 1>}
  NamedRerouteDeclaration(None=AppendVector(A=ScalarParameter() {defaultValue=0.156, parameterName=Point3_x}, B=ScalarParameter() {defaultValue=-0.8, parameterName=Point3_y})) {name=Point3_1, nodeColor=<rgba 0 0 0 1>}
 ]
}
