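# Translucent UI material that draws the Tow Hook weapon's reticle in white over a soft black outer glow.
# Smoothstep thresholds on the reticle texture's red channel, treated like a distance field, cut the line and glow masks.
# ADS_ThicknessSwitch blends to lower, thicker thresholds while changing the centered UV scale from 1 to half.
# Opacity adds the glow, weighted by OuterGlowOpacity, to the line mask, then multiplies by OuterReticleOpacity.
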Material /TowHookWeapon/UI/Materials/M_UI_TowHookReticle
graph {
 properties {blendMode=BLEND_Translucent, materialDomain=MD_UI, pixelDepthOffsetMode=PDOM_Legacy, usedShadingModels=Default Lit}
 nodes [
  n1 = ScalarParameter() {parameterName=ADS_ThicknessSwitch}
  n2 = TextureSample(UVs=MaterialFunctionCall(Texture Scale=LinearInterpolate(Alpha=@n1) {constA=1, constAlpha=0, constB=0.5}) {materialFunction=&ScaleUVsByCenter}.UVs) {samplerSource=SSM_Clamp_WorldGroupSettings, texture=/TowHookWeapon/UI/Textures/T_UI_Reticle_TowHook}
  n3 = SmoothStep(Value=@n2.R) {constMax=0.4, constMin=0.35}
  n4 = SmoothStep(Value=@n2.R) {constMax=0.3, constMin=0.25}
  LinearInterpolate(
   A=Constant3Vector() {constant=<rgba 0 0 0 1>}
   B=Constant3Vector() {constant=<rgba 1 1 1 1>}
   Alpha=LinearInterpolate(A=@n3, B=@n4, Alpha=ScalarParameter() {parameterName=ADS_ThicknessSwitch})
  )
  n5 = Multiply(A=@n4, B=ScalarParameter() {defaultValue=0.5, parameterName=OuterGlowOpacity}) {constB=0.3}
  n6 = Multiply(A=SmoothStep(Value=@n2.R) {constMax=0.2, constMin=0.15}, B=ScalarParameter() {defaultValue=0.5, parameterName=OuterGlowOpacity}) {constB=0.3}
  Multiply(
   A=LinearInterpolate(A=Add(A=@n3, B=@n5), B=Add(A=@n4, B=@n6), Alpha=ScalarParameter() {parameterName=ADS_ThicknessSwitch})
   B=ScalarParameter() {defaultValue=1, group=Opacity, parameterName=OuterReticleOpacity}
  )
 ]
}
